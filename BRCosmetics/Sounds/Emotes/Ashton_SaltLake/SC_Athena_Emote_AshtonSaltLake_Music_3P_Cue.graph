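# Third-person music cue for the Ashton SaltLake emote in Battle Royale, heard by players near the emoting character.
# A single looping wave player feeds the output, shaped by the shared emote attenuation, 3P music sound class and property sheet.
# It joins the emote music concurrency group and keeps playing when silent to hold its place in the loop.
# A submix send routes it to music audio analysis.
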SoundCue /BRCosmetics/Sounds/Emotes/Ashton_SaltLake/SC_Athena_Emote_AshtonSaltLake_Music_3P_Cue
root=WavePlayer_1
nodes [{id=WavePlayer_1, properties {bLooping=true, soundWaveAssetPtr=&/Emote_AshtonSaltLake_01}}]
properties {
 attenuationSettings=&Fort_Emotes_Athena_Att
 audioPropertiesSheet=/BRCosmetics/Sounds/Settings/PropertySheets/Emotes/APS_Cosmetic_Emote_Music_3P
 maxDistance=1600
 soundClassObject=/Game/Sounds/FortSoundClassesAndModes/Fort_Music_Emote_3P
 virtualizationMode=PlayWhenSilent
 volumeMultiplier=1.75
 concurrencyOverrides {resolutionRule=PreventNew}
 concurrencySet [/Game/Sounds/Concurrency_Rules/Athena_EmoteMusic_Conc]
 soundSubmixSends [{minSendLevel=0, soundSubmix=&Fortnite_AudioAnalysis_Music}]
}
